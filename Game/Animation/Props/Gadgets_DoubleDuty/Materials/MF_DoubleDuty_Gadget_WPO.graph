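# Material function for the DoubleDuty gadget prop that outputs a pulsing squish offset, named for use as World Position Offset.
# A sine of time scaled by WPOSquishSpeed drives the pulse, with WPOSquishIntensity setting its amplitude.
# The result multiplies the world-space vertex normal with Z masked out, so vertices swell and shrink only along world X and Y.
# The phase has no positional term, so all vertices pulse in sync.
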MaterialFunction /Game/Animation/Props/Gadgets_DoubleDuty/Materials/MF_DoubleDuty_Gadget_WPO
graph {nodes [
  n1 = Time()
  n2 = ScalarParameter() {defaultValue=2, parameterName=WPOSquishSpeed}
  n3 = Multiply(A=Constant3Vector() {constant=<rgba 1 1 0 0>}, B=Sine(None=Multiply(A=@n1, B=@n2)))
  n4 = ScalarParameter() {defaultValue=0.5, parameterName=WPOSquishIntensity}
  FunctionOutput(None=Multiply(A=Multiply(A=@n3, B=@n4), B=VertexNormalWS()))
 ]}
