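# Translucent lit material for the DoubleDuty gadget prop, drawing a two-color, wind-like swirl with fresnel rim and soft edges.
# A mask texture pans in opposing directions with a vertical-normal pole mask, driving noise, color, opacity and screen-UV distortion.
# Higher quality tiers composite time-of-day-tinted color over distorted SceneColor at full opacity; mobile tiers skip distortion and use computed opacity.
# A material function supplies world position offset.
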Material /Game/Animation/Props/Gadgets_DoubleDuty/Materials/M_DoubleDuty_Gadget
graph {
 properties {
  bUsedWithInstancedStaticMeshes=true
  bUsedWithMeshDeformer=true
  bUsedWithNanite=true
  bUsedWithNiagaraMeshParticles=true
  bUsedWithSkeletalMesh=true
  blendMode=BLEND_Translucent
  pixelDepthOffsetMode=PDOM_Legacy
  refractionDepthBias=1
  usedShadingModels=Default Lit
 }
 nodes [
  MaterialFunctionCall() {materialFunction=&/MF_DoubleDuty_Gadget_WPO}
  n1 = ScreenPosition()
  n2 = NamedRerouteUsage()
  n3 = MaterialFunctionCall(Color=NamedRerouteUsage()) {materialFunction=&SplitComponents}
  n4 = Saturate(None=Subtract(
    A=MaterialFunctionCall(Color=@n2) {materialFunction=&SplitComponents}.B
    B=Multiply(A=@n3.G) {constB=0.5}
   ))
  n5 = LinearInterpolate(A=ScalarParameter() {defaultValue=0.45, parameterName=Param_5}, Alpha=@n4) {constA=0.5}
  n6 = Multiply(
   A=Multiply(A=Subtract(A=@n5) {constB=0.5}, B=ScalarParameter() {defaultValue=0.025, parameterName=DistortionIntensity})
   B=Power(Base=NamedRerouteUsage(), Exp=ScalarParameter() {defaultValue=0.2, parameterName=Param_6})
  )
  n7 = Multiply(
   A=SmoothStep(Min=ScalarParameter() {defaultValue=0.4, parameterName=ExtraNoisePower}, Value=@n4) {constMin=0.3}
   B=ScalarParameter() {defaultValue=0.55, parameterName=ExtraNoiseIntensity}
  ) {constB=0.65}
  n8 = NamedRerouteUsage()
  n9 = NamedRerouteUsage()
  n10 = Saturate(None=Add(
    A=MaterialFunctionCall(Color=@n8) {materialFunction=&SplitComponents}.R
    B=MaterialFunctionCall(Color=@n9) {materialFunction=&SplitComponents}.G
   ))
  n11 = Multiply(
   A=Power(Base=@n10, Exp=ScalarParameter() {defaultValue=1, parameterName=MainWindPower})
   B=ScalarParameter() {defaultValue=0.5, parameterName=MainWindIntensity}
  )
  n12 = ScalarParameter() {defaultValue=0.2, parameterName=Param_1}
  n13 = Fresnel() {baseReflectFraction=0, exponent=0.1}
  n14 = ScalarParameter() {defaultValue=5, parameterName=Param_2}
  n15 = SmoothStep(
   Min=Multiply(A=OneMinus(None=@n10), B=ScalarParameter() {defaultValue=0.75}) {constB=0.5}
   Max=ScalarParameter() {defaultValue=5, parameterName=Param_4}
   Value=Multiply(A=Power(Base=@n13, Exp=@n14) {constExponent=4}, B=ScalarParameter() {defaultValue=2, parameterName=Param_3})
  ) {constMax=5}
  n16 = Saturate(None=LinearInterpolate(
    A=Add(A=@n7, B=@n11)
    B=ScalarParameter() {defaultValue=1, parameterName=RimIntensity}
    Alpha=SmoothStep(Max=@n12, Value=@n15) {constMax=0.2}
   ))
  n17 = ScalarParameter() {defaultValue=0.9, parameterName=FrontOpacity}
  n18 = LinearInterpolate(
   A=VectorParameter() {parameterName=SecondaryColor, defaultValue=<rgba 0.0441 0.3392 0.6042 1>}.RGB
   B=VectorParameter() {parameterName=MainColor, defaultValue=<rgba 1.1456 1.2607 1.3 1>}.RGB
   Alpha=Multiply(A=@n16, B=@n17)
  )
  n19 = MaterialFunctionCall(Color_Morning=@n18, Color_Day=@n18, Color_Evening=Multiply(A=@n18) {constB=0.75}, Color_Night=Multiply(A=@n18) {constB=0.65}) {materialFunction=&Color_Time_of_Day}
  n20 = ScalarParameter() {defaultValue=6, parameterName=InnerFresnelPower}
  n21 = Saturate(None=Add(A=Power(Base=@n13, Exp=@n20) {constExponent=8}, B=ScalarParameter() {defaultValue=0.4, parameterName=InnerInfill}) {constB=0.25})
  n22 = Multiply(
   A=LinearInterpolate(B=Multiply(A=@n21, B=@n5), Alpha=NamedRerouteUsage()) {constA=0.5}
   B=ScalarParameter() {defaultValue=0.7, parameterName=BackgroundOpacity}
  ) {constA=1, constB=0.5}
  n23 = Fresnel(ExponentIn=ScalarParameter() {defaultValue=2.5, parameterName=EdgeFresnelPower}) {baseReflectFraction=0, exponent=3}
  n24 = Multiply(
   A=Multiply(A=LinearInterpolate(A=@n22, B=@n17, Alpha=@n16), B=ScalarParameter() {defaultValue=0.9, parameterName=OverallOpacity})
   B=SmoothStep(Value=OneMinus(None=@n23))
  )
  n25 = LinearInterpolate(A=SceneColor!(UVs=Add(A=@n1.ViewportUV, B=@n6)).RGB, B=@n19.Result, Alpha=@n24)
  MaterialFunctionCall!(
   Low _ PC Low and Med=@n25
   "High _ XB1, PS4"=@n25
   "Epic _ PC High, XSX, PS5"=@n25
   --------
   "Mobile Low _ PC Perf, Android"=@n19.Result
   Mobile Med _ Switch=@n19.Result
  ) {materialFunction=/Game/Global/MaterialFunctions/TechArt/MF_Fortnite_QualitySwitch_Vector3}
  n26 = Constant() {r=1}
  MaterialFunctionCall!(Low=@n26, High=@n26, Epic=@n26, --------, Mobile=@n24, Medium=@n24) {materialFunction=/FortniteBaseMaterials/Functions/MF_QualitySwitch_Scalar}
  n27 = Multiply(A=TextureCoordinate(), B=ComponentMask(None=VectorParameter() {defaultValue=<rgba 1 1.3 0 1>}.RGB) {g=true, r=true})
  n28 = NamedRerouteDeclaration(None=Panner(Coordinate=@n27) {speedX=2.5, speedY=-0.3}) {name=UV_Left, nodeColor=<rgba 1 0 0.5859 1>}
  NamedRerouteDeclaration(None=Multiply(A=TextureSample(UVs=@n28) {texture=/Game/Animation/Props/Gadgets_DoubleDuty/Textures/T_DoubleDuty_Mask}.RGB, B=NamedRerouteUsage())) {name=Left, nodeColor=<rgba 0 0.8594 1 1>}
  n29 = Panner(Coordinate=@n27) {speedX=-2, speedY=0.5}
  NamedRerouteDeclaration(None=Multiply(A=TextureSample(UVs=@n29) {texture=&T_DoubleDuty_Mask}.RGB, B=NamedRerouteUsage())) {name=Right, nodeColor=<rgba 0.6172 0 1 1>}
  n30 = Abs(None=ComponentMask(None=VertexNormalWS()) {b=true})
  NamedRerouteDeclaration(None=OneMinus(None=Power(Base=@n30) {constExponent=6})) {name=PolarsMask, nodeColor=<rgba 1 0.2578 0 1>}
 ]
}
