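# Masked unlit shadow material for the DoubleDuty gadget prop, set up for skeletal meshes and mesh deformers.
# A ShadowReplace node gives zero in the normal pass and a temporal-AA dither driven by the ShadowOpacity parameter in shadow passes, which points to an invisible mesh casting a partial shadow.
# It also includes the shared MF_DoubleDuty_Gadget_WPO function, named for world position offset.
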Material /Game/Animation/Props/Gadgets_DoubleDuty/Materials/M_DoubleDuty_Gadget_Shadow
graph {
 properties {
  bUsedWithMeshDeformer=true
  bUsedWithSkeletalMesh=true
  blendMode=BLEND_Masked
  pixelDepthOffsetMode=PDOM_Legacy
  shadingModel=MSM_Unlit
  usedShadingModels=Unlit
 }
 nodes [
  MaterialFunctionCall() {materialFunction=&/MF_DoubleDuty_Gadget_WPO}
  ShadowReplace(
   Default=Constant()
   Shadow=MaterialFunctionCall(Alpha Threshold=ScalarParameter() {defaultValue=0.3, parameterName=ShadowOpacity}) {materialFunction=&DitherTemporalAA}.Result
  )
 ]
}
